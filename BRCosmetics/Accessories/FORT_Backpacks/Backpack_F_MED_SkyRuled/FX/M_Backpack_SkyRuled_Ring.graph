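# Unlit, two-sided translucent material for the galaxy-styled ring effect on the SkyRuled backpack, built for Niagara mesh particles.
# Emissive layers panning stars, tinted fire-noise bands and purple gradients over a screen-aligned nebula; UseDarkRingNoise switches to particle color with a noise mask.
# Opacity combines UV band, range and side-fade masks with depth and view-angle fades.
# SplineThicken, scaled by a dynamic Width parameter, plus a sine-tapered tangent offset shape the mesh.
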Material /BRCosmetics/Accessories/FORT_Backpacks/Backpack_F_MED_SkyRuled/FX/M_Backpack_SkyRuled_Ring
graph {
 properties {
  bAutomaticallySetUsageInEditor=false
  bUsedWithNiagaraMeshParticles=true
  blendMode=BLEND_Translucent
  shadingModel=MSM_Unlit
  twoSided=true
  usedShadingModels=Unlit
 }
 nodes [
  n1 = ScalarParameter() {defaultValue=6.5, parameterName=Width}
  n2 = DynamicParameter() {paramNames [Width,Param2,Param3,Param4]}
  n3 = TextureCoordinate()
  n4 = Multiply(
   A=Subtract(A=ComponentMask(None=@n3) {g=true}) {constB=0.5}
   B=MaterialFunctionCall() {materialFunction=&TangentBasis}.Y V
  )
  n5 = Sine(None=ComponentMask(None=TextureCoordinate()) {r=true}) {period=2}
  n6 = LinearInterpolate(Alpha=Saturate(None=Multiply(A=@n5) {constB=1.5})) {constA=0.1}
  Add(
   A=MaterialFunctionCall(WidthBase=Multiply(A=@n1, B=@n2.Width), WidthTip=Multiply(A=@n1, B=@n2.Width), Expand U or V UV Channel=StaticBool()) {materialFunction=&SplineThicken}.WorldPositionOffset
   B=Multiply(A=Multiply(A=@n4, B=@n6), B=ScalarParameter() {parameterName=PlaneWidth})
  )
  n7 = SphereMask(A=TextureCoordinate() {vTiling=0}, B=Constant2Vector() {r=0.8}) {attenuationRadius=0.2, hardnessPercent=80}
  n8 = Saturate(None=OneMinus(None=Saturate(None=@n7)))
  n9 = ScalarParameter() {parameterName=RemoveRingBackMask}
  n10 = TextureCoordinate()
  n11 = Power(Base=Sine(None=ComponentMask(None=@n10) {g=true}) {period=2}) {constExponent=5}
  n12 = TextureCoordinate()
  n13 = ComponentMask(None=VectorParameter() {parameterName=DarkNoiseUV, defaultValue=<rgba 3 1.5 0 0>}.RGB) {g=true, r=true}
  n14 = VectorParameter() {parameterName=DarkNoiseSpeed, defaultValue=<rgba 0.15 0.1 0 0>}
  n15 = TextureSample(UVs=Panner(Coordinate=Multiply(A=@n12, B=@n13), Speed=ComponentMask(None=@n14.RGB) {g=true, r=true}) {speedX=0.15, speedY=0.2}) {texture=/BRCosmetics/Weapons/FORT_Melee/Pickaxe_CoatCheck/FX/T_Pickaxe_CoatCheck_Noise}
  n16 = Multiply(
   A=Saturate(None=Multiply(A=@n11))
   B=Power(Base=Multiply(A=@n15.R), Exp=ScalarParameter() {defaultValue=0.6, parameterName=DarkNoisePower}) {constExponent=0.6}
  )
  n17 = ComponentMask(None=TextureCoordinate()) {g=true}
  n18 = Saturate(None=Multiply(A=Power(Base=@n17) {constExponent=3}) {constB=20})
  NamedRerouteDeclaration(None=Multiply(A=Multiply(A=@n16, B=@n18), B=LinearInterpolate(A=@n8, Alpha=@n9))) {name=DarkRingNoise, nodeColor=<rgba 1 0.8906 0 1>}
  n19 = TextureSample(UVs=MaterialFunctionCall!(TextureResolution=Constant() {r=160}) {materialFunction=/Engine/Functions/Engine_MaterialFunctions02/Texturing/ScreenAlignedPixelToPixelUVs}.Result) {texture=/Game/Characters/Player/Male/Medium/Bodies/M_Med_Soldier_04/Skins/Galaxy/Textures/T_FN_Nebula_style2}
  n20 = OneMinus(None=ComponentMask(None=TextureCoordinate()) {g=true})
  n21 = MaterialFunctionCall(
   Input=Saturate(None=Subtract(A=@n20) {constB=0.65})
   Input Low=ScalarParameter()
   Input High=ScalarParameter() {defaultValue=1, parameterName=Param_1}
  ) {materialFunction=&RemapValue_0-1}
  n22 = ScalarParameter() {defaultValue=5, parameterName=Param_2}
  n23 = LinearInterpolate(
   B=LinearInterpolate(Alpha=Multiply(A=@n21.Result, B=@n22) {constB=5}) {constA=1, constB=5}
   Alpha=ScalarParameter() {parameterName=UseCenterBright}
  ) {constA=1}
  n24 = TextureCoordinate() {uTiling=10, vTiling=2}
  n25 = VectorParameter() {parameterName=StarSpeed, defaultValue=<rgba 0.5 0 0 0>}
  n26 = TextureSample(UVs=Panner(Coordinate=Multiply(A=@n24) {constB=0.9}, Speed=ComponentMask(None=@n25.RGB) {g=true, r=true}) {speedX=0.5}) {samplerType=SAMPLERTYPE_Masks, texture=/Game/Characters/Player/Male/Medium/Bodies/M_Med_Soldier_04/Skins/Galaxy/Textures/Celestial_T_Stars}
  n27 = Multiply(A=@n26.B) {constB=2}
  n28 = TextureSample(UVs=TextureCoordinate() {uTiling=10, vTiling=3}) {samplerType=SAMPLERTYPE_Masks, texture=&Celestial_T_Stars}
  n29 = Saturate(None=LinearInterpolate(Alpha=Multiply(A=@n28.G)) {constA=-1.2, constB=3})
  n30 = TextureCoordinate()
  n31 = ComponentMask(None=VectorParameter() {parameterName=CenterSideUV, defaultValue=<rgba 5 5 0 0>}.RGB) {g=true, r=true}
  n32 = VectorParameter() {parameterName=CenterSideSpeed, defaultValue=<rgba 0.7 0.3 0 0>}
  n33 = TextureSampleParameter2D(UVs=Panner(Coordinate=Multiply(A=@n30, B=@n31), Speed=ComponentMask(None=@n32.RGB) {g=true, r=true}) {speedX=0.7, speedY=0.3}) {parameterName=CenterTex, texture=/BRCosmetics/Athena/Effects/Particles/Skydiving/NS_Contrail_GrimHound/T_FX_Contrail_GrimHound_Fire}
  n34 = Multiply(
   A=VectorParameter() {parameterName=CenterTex_Color, defaultValue=<rgba 1 0.2672 0.3789 1>}.RGB
   B=Multiply(A=Power(Base=@n33.B) {constExponent=0.5}) {constB=2}
  )
  n35 = Add(A=TextureCoordinate()) {constB=0.35}
  n36 = Subtract(A=Sine(None=ComponentMask(None=@n35) {g=true}) {period=2}) {constB=0.8}
  n37 = Saturate(None=Multiply(A=Power(Base=@n36)) {constB=50})
  n38 = VectorParameter() {parameterName=ColorSideUV, defaultValue=<rgba 2 2 0 0>}
  n39 = Panner(
   Coordinate=Multiply(A=TextureCoordinate(), B=ComponentMask(None=@n38.RGB) {g=true, r=true})
   Speed=ComponentMask(None=VectorParameter() {parameterName=ColorSideSpeed, defaultValue=<rgba 0.15 0.2 0 0>}.RGB) {g=true, r=true}
  ) {speedX=0.15, speedY=0.2}
  n40 = Power(Base=Multiply(A=TextureSampleParameter2D(UVs=@n39) {parameterName=Color_Tex, texture=&T_FX_Contrail_GrimHound_Fire}.B)) {constExponent=1}
  n41 = ComponentMask(None=Add(A=TextureCoordinate()) {constB=0.2}) {g=true}
  n42 = Power(Base=Subtract(A=Sine(None=@n41) {period=2}) {constB=0.8})
  n43 = Multiply(
   A=Multiply(A=VectorParameter() {parameterName=ColorSideColor, defaultValue=<rgba 1.5485 0.4424 2 1>}.RGB, B=Multiply(A=@n40) {constB=0.5})
   B=Saturate(None=Multiply(A=@n42) {constB=50})
  )
  n44 = ComponentMask(None=TextureCoordinate()) {g=true}
  n45 = Power(Base=LinearInterpolate(Alpha=OneMinus(None=@n44)) {constA=-0.3}) {constExponent=2.5}
  n46 = LinearInterpolate(
   A=Constant4Vector() {constant=<rgba 0.0524 0 0.1042 1>}.RGB
   B=Constant4Vector() {constant=<rgba 0.781 0 0.849 1>}.RGB
   Alpha=Saturate(None=Multiply(A=@n45) {constB=5})
  )
  n47 = VectorParameter() {parameterName=DarkOutSideUV, defaultValue=<rgba 3 3 0 0>}
  n48 = Panner(
   Coordinate=Multiply(A=TextureCoordinate(), B=ComponentMask(None=@n47.RGB) {g=true, r=true})
   Speed=ComponentMask(None=VectorParameter() {parameterName=DarkOutSideSpeed, defaultValue=<rgba 1 0.1 0 0>}.RGB) {g=true, r=true}
  ) {speedX=1, speedY=0.1}
  n49 = Power(Base=Multiply(
    A=VectorParameter() {parameterName=DarkOutColor, defaultValue=<rgba 0.089 0 1.2 1>}.RGB
    B=TextureSample(UVs=@n48) {texture=&T_FX_Contrail_GrimHound_Fire}.R
   )) {constExponent=1}
  n50 = OneMinus(None=ComponentMask(None=TextureCoordinate()) {g=true})
  n51 = Saturate(None=Multiply(A=Subtract(A=@n50) {constB=0.3}) {constB=5})
  n52 = Constant4Vector() {constant=<rgba 0.0079 0.0002 0.1 1>}
  n53 = Constant4Vector() {constant=<rgba 0.0045 0.0001 0.0573 1>}
  n54 = ComponentMask(None=TextureCoordinate()) {g=true}
  n55 = Saturate(None=LinearInterpolate(Alpha=Power(Base=@n54) {constExponent=1}) {constA=-0.5})
  n56 = Add(A=Multiply(A=Multiply(A=@n46)), B=Add(A=Multiply(A=@n49, B=@n51), B=LinearInterpolate(A=@n52.RGB, B=@n53.RGB, Alpha=@n55)))
  n57 = Add(
   A=Multiply(A=Multiply(A=@n27, B=@n29), B=Constant4Vector() {constant=<rgba 2.6484 1.2422 3 1>}.RGB) {constB=3}
   B=Add(A=Multiply(A=@n34, B=@n37), B=Add(A=@n43, B=@n56))
  )
  n58 = ParticleColor()
  n59 = Multiply(A=LinearInterpolate(A=Multiply(A=@n19.RGB) {constB=2}, B=Multiply(A=@n57, B=@n23), Alpha=Constant() {r=0.8}) {constAlpha=1}, B=@n58.RGB)
  n60 = ParticleColor()
  n61 = ScalarParameter() {parameterName=UseDarkRingNoise}
  Reroute(None=MaterialFunctionCall(Emissive=LinearInterpolate(A=@n59, B=@n60.RGB, Alpha=@n61)) {materialFunction=&MF_FX_EmissiveAdjustments}.Result)
  n62 = TextureCoordinate()
  n63 = Subtract(A=OneMinus(None=ComponentMask(None=@n62) {g=true})) {constB=0.55}
  n64 = Saturate(None=Multiply(A=Power(Base=@n63) {constExponent=1}) {constB=10})
  n65 = LinearInterpolate(
   B=LinearInterpolate(A=@n64, B=OneMinus(None=@n64), Alpha=ScalarParameter() {parameterName=InvertRangeMask})
   Alpha=ScalarParameter() {parameterName=UseRangeMask}
  ) {constA=1}
  n66 = TextureCoordinate()
  n67 = Power(Base=Sine(None=ComponentMask(None=@n66) {g=true}) {period=2}) {constExponent=5}
  n68 = Multiply(A=Saturate(None=Multiply(A=@n67)))
  n69 = TextureCoordinate()
  n70 = Subtract(A=OneMinus(None=ComponentMask(None=@n69) {g=true})) {constB=0.5}
  n71 = Saturate(None=Multiply(A=Power(Base=@n70)) {constB=5})
  n72 = Subtract(A=ComponentMask(None=TextureCoordinate()) {g=true}) {constB=0.03}
  n73 = Saturate(None=Multiply(A=Power(Base=@n72)) {constB=20})
  n74 = Add(A=@n68, B=LinearInterpolate(B=Multiply(A=@n71, B=@n73)) {constAlpha=0.3}) {constB=0}
  n75 = LinearInterpolate(
   A=LinearInterpolate(A=@n68, B=Saturate(None=@n74), Alpha=ScalarParameter() {parameterName=ControlCenterMask})
   B=NamedRerouteUsage()
   Alpha=ScalarParameter() {parameterName=UseDarkRingNoise}
  )
  n76 = ComponentMask(None=TextureCoordinate()) {r=true}
  n77 = Multiply(
   A=Power(Base=Sine(None=@n76) {period=2}, Exp=ScalarParameter() {defaultValue=1.5, parameterName=UseLeft and Right SideFade Power}) {constExponent=1.5}
   B=ScalarParameter() {defaultValue=1.5, parameterName=UseLeft and Right SideFade Edge}
  ) {constB=1.5}
  n78 = Multiply(
   A=Multiply(A=@n58.A, B=Multiply(A=@n75, B=@n65))
   B=LinearInterpolate(B=Saturate(None=@n77), Alpha=ScalarParameter() {parameterName=UseLeft and Right SideFade}) {constA=1}
  )
  n79 = ScalarParameter() {parameterName=FadeDistance}
  n80 = Power(Base=MaterialFunctionCall("FadeSensitivity (optional)"=ScalarParameter() {parameterName="FadeSensitivity (optional) (S)"}) {materialFunction=&FadeBasedOnViewAngle}.Result) {constExponent=1.3}
  n81 = Constant() {r=1}
  Reroute(None=Multiply(
    A=DepthFade(Opacity=@n78, FadeDistance=@n79)
    B=MaterialFunctionCall(In=@n80, Contrast=@n81) {materialFunction=&CheapContrast}.Result
   ))
  Round(None=@n18)
  n82 = TextureCoordinate()
  Round(None=Add(A=ComponentMask(None=@n82) {r=true}) {constB=-0.2})
  n83 = TextureCoordinate()
  n84 = Subtract(A=OneMinus(None=ComponentMask(None=@n83) {g=true})) {constB=0.45}
  Saturate(None=Multiply(A=Power(Base=@n84) {constExponent=1}) {constB=10})
 ]
}
