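# Niagara effect for the SkyRuled backpack cosmetic: a glowing, layered ring effect with orbiting mesh debris.
# Four burst emitters build persistent ring, orbit and core layers from meshes, most using a local CalcCameraFacing module, with sprites and a light at the center.
# A GPU emitter continuously spawns small stones driven by vortex, attraction and drag, sized by a time-smoothed system value.
# Shared System.FXOffset places the layers, and System.SizeAll scales the rings.
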NiagaraSystem /BRCosmetics/Accessories/FORT_Backpacks/Backpack_F_MED_SkyRuled/FX/NS_Backpack_SkyRuled
dependencies {
 usedDataInterfaces [
  &NiagaraDataInterfaceColorCurve
  &NiagaraDataInterfaceCamera
  &NiagaraDataInterfaceSkeletalMesh
  &NiagaraDataInterfaceCurve
 ]
 usedDynamicInputs [
  &ColorFromCurve
  &RandomRangeFloat
  &ReturnNormalizedExecIndex
  &MakeCustomFloatFromBool
  &SetBoolByFloatComparison
  &VectorLength
 ]
 usedModules [
  &self:SystemSpawnScript.SystemScriptSource.SystemScriptGraph.NiagaraNodeAssignment_0.SetVariables_#1
  &self:SystemSpawnScript.SystemScriptSource.SystemScriptGraph.NiagaraNodeAssignment_1.SetVariables_#2
  &SystemState
  &InterpolateOverTime
 ]
}
emitters {
 BackgorundRing {
  emitterUpdateScript [
   &EmitterState(Life Cycle Mode=::NewEnumerator1, Loop Behavior=::NewEnumerator1, Loop Duration Mode=::NewEnumerator1, UseLoopDelay=false)
   &SpawnBurst_Instantaneous(Use Spawn Probability=false, UseLoopCountLimit=false)
  ]
  particleSpawnScript [
   &InitializeParticle(
    Mesh Scale Mode=::NewEnumerator3
    Mesh Uniform Scale=0.22
    Position Offset=$System.FXOffset
    UsePositionOffset=true
    Write Material Random=true
   )
   &InitialMeshOrientation(Mesh Orientation Mode=::NewEnumerator1, Use Rotation Vector=true) {module=InitialMeshOrientation003}
   &InitialMeshOrientation(Mesh Orientation Mode=::NewEnumerator7, Rotation=<0.13 0 0>, Use Rotation Vector=true) {module=InitialMeshOrientation004}
   &InitialMeshOrientation(Mesh Orientation Mode=::NewEnumerator7, Use Rotation Vector=true) {module=InitialMeshOrientation005}
  ]
  particleUpdateScript [
   &ParticleState(Kill Particles When Lifetime Has Elapsed=false)
   &ScaleMeshSize(Scale Factor=&Multiply_VectorByFloat(Vector=&Multiply_VectorByFloat(Vector=<1.7 1 0.7>, Float=0.9), Float=$System.SizeAll))
   &self:CalcCameraFacing(Camera Query=DataInterface() {propertyValues {PlayerControllerIndex=0, bRequireCurrentFrameData=false}}) {module=CalcCameraFacing}
  ]
  rendererClasses [&NiagaraMeshRendererProperties]
  renderers [{rendererClass=&NiagaraMeshRendererProperties, rendererIndex=0}]
 }
 CenterRing {
  emitterUpdateScript [
   &EmitterState(Life Cycle Mode=::NewEnumerator1, Loop Behavior=::NewEnumerator1, Loop Duration Mode=::NewEnumerator1, UseLoopDelay=false)
   &SpawnBurst_Instantaneous(Use Spawn Probability=false, UseLoopCountLimit=false)
  ]
  particleSpawnScript [
   &InitializeParticle(
    Color Mode=::NewEnumerator0
    Mesh Scale Mode=::NewEnumerator3
    Mesh Uniform Scale=0.2
    Position Offset=$System.FXOffset
    UsePositionOffset=true
    Write Material Random=true
   )
   &InitialMeshOrientation(Mesh Orientation Mode=::NewEnumerator2, Rotation=<0 0 -0.25>, Use Rotation Vector=true)
   &DynamicMaterialParameters(
    Index 2 Param 1=1
    Index 3 Param 1=2
    Index 3 Param 2=5
    Module.Index2Float0Write=true
    Module.Index2Float1Write=true
    Module.Index2Float2Write=true
    Module.Index2Float3Write=true
    Module.Index3Float0Write=true
    Module.Index3Float1Write=true
    Module.Index3Float2Write=true
    Module.Index3Float3Write=true
    Write Parameter Index 0=::NewEnumerator0
    Write Parameter Index 2=::NewEnumerator1
    Write Parameter Index 3=::NewEnumerator1
   )
  ]
  particleUpdateScript [
   &ParticleState(Kill Particles When Lifetime Has Elapsed=false)
   &SpriteRotationRate(Rotation Rate=80)
   &ScaleMeshSize(Scale Factor=&Multiply_VectorByFloat(Vector=<1 1 1>, Float=$System.SizeAll))
   &FN_Sprite_Attributes(
    Color=&MakeLinearColorFromVectorAndFloat(
     "Vector (RGB)"=&MakeVectorFromLinearColorRGB(Color=<rgba 0 0 0.2135 7>)
     "Float (Alpha)"=&Multiply_Float(A=7, B=1)
    )
    Create Color Attribute=true
    Create SpriteSize Attribute=true
    Module Output:ENiagara_AttributeSamplingApplyOutput=::NewEnumerator1
    SpriteSize=&Vector2DFromFloat(Value=&Multiply_Float(
      A=46
      B=&Multiply_Float(A=$System.SizeAll, B=&MakeFloatFromVector(Vector=$Engine.Owner.Scale, Channel=::NewEnumerator0))
     ))
   )
   &FN_Sprite_Attributes(
    Camera Offset=10
    Color=&MakeLinearColorFromVectorAndFloat(
     "Vector (RGB)"=&MakeVectorFromLinearColorRGB(Color=<rgba 3.7917 12.3035 40 2>)
     "Float (Alpha)"=&Multiply_Float(A=2, B=0.5)
    )
    Create Camera Offset Attribute=true
    Create Color Attribute=true
    Create Position Attribute=true
    Create SpriteSize Attribute=true
    Module Output:ENiagara_AttributeSamplingApplyOutput=::NewEnumerator1
    Position:NiagaraPosition=&AddVectorToPosition(Position=Linked() {linkedVariable {name=System.FXOffset, type=Vector3f}}, Vector=<0 0 0>)
    SpriteSize=&Vector2DFromFloat(Value=&Multiply_Float(
      A=26
      B=&Multiply_Float(A=$System.SizeAll, B=&MakeFloatFromVector(Vector=$Engine.Owner.Scale, Channel=::NewEnumerator0))
     ))
   ) {module=FN_Sprite_Attributes001}
   &FN_Sprite_Attributes(
    Camera Offset=10
    Color=&MakeLinearColorFromVectorAndFloat("Vector (RGB)"=&MakeVectorFromLinearColorRGB(Color=<rgba 0 0 0 1>), "Float (Alpha)"=10)
    Create Camera Offset Attribute=true
    Create Color Attribute=true
    Create Position Attribute=true
    Create SpriteSize Attribute=true
    Module Output:ENiagara_AttributeSamplingApplyOutput=::NewEnumerator1
    Position:NiagaraPosition=&AddVectorToPosition(Position=Linked() {linkedVariable {name=System.FXOffset, type=Vector3f}}, Vector=<0 0 0>)
    SpriteSize=&Vector2DFromFloat(Value=&Multiply_Float(
      A=21
      B=&Multiply_Float(A=$System.SizeAll, B=&MakeFloatFromVector(Vector=$Engine.Owner.Scale, Channel=::NewEnumerator0))
     ))
   ) {module=FN_Sprite_Attributes002}
   &Light_Attributes(Color=1, Color Mode=0, Falloff Exponent=1, Light Color:LinearColor=<rgba 5.1464 0.5208 10 1>, Light Exponent=3, Light Radius=60, Radius=1)
  ]
  rendererClasses [
   &NiagaraMeshRendererProperties
   &NiagaraSpriteRendererProperties
   &NiagaraLightRendererProperties
  ]
  renderers [
   {rendererClass=&NiagaraMeshRendererProperties, rendererIndex=0}
   {rendererClass=&NiagaraSpriteRendererProperties, rendererIndex=1}
   {rendererClass=&NiagaraSpriteRendererProperties, rendererIndex=2}
   {rendererClass=&NiagaraSpriteRendererProperties, rendererIndex=3}
   {rendererClass=&NiagaraLightRendererProperties, rendererIndex=4}
  ]
 }
 ColorRing {
  emitterUpdateScript [
   &EmitterState(Life Cycle Mode=::NewEnumerator1, Loop Behavior=::NewEnumerator1, Loop Duration Mode=::NewEnumerator1, UseLoopDelay=false)
   &SpawnBurst_Instantaneous(Use Spawn Probability=false, UseLoopCountLimit=false)
  ]
  particleSpawnScript [
   &InitializeParticle(
    Color=<rgba 5 1.0482 2.4429 2>
    Mesh Scale Mode=::NewEnumerator3
    Position Offset=&AddVectorToPosition(Position=Linked() {linkedVariable {name=System.FXOffset, type=Vector3f}}, Vector=<0 0 0>)
    UsePositionOffset=true
    Write Material Random=true
   )
   &InitialMeshOrientation(Mesh Orientation Mode=::NewEnumerator1, Rotation=<0.5 0 0>, Use Rotation Vector=true) {module=InitialMeshOrientation003}
   &InitialMeshOrientation(Mesh Orientation Mode=::NewEnumerator7, Rotation=<0.1 0 0>, Use Rotation Vector=true) {module=InitialMeshOrientation004}
   &InitialMeshOrientation(Mesh Orientation Mode=::NewEnumerator7, Use Rotation Vector=true) {module=InitialMeshOrientation005}
   &DynamicMaterialParameters(
    Index 0 Param 1=7
    Module.Index0Float0Write=true
    Module.Index0Float1Write=false
    Module.Index0Float2Write=false
    Module.Index0Float3Write=false
   )
  ]
  particleUpdateScript [
   &ParticleState(Kill Particles When Lifetime Has Elapsed=false)
   &ScaleMeshSize(Scale Factor=&Multiply_VectorByFloat(Vector=&Multiply_VectorByFloat(Vector=<1.5 1 0.6>, Float=0.9), Float=$System.SizeAll))
   &self:CalcCameraFacing(Camera Query=DataInterface() {propertyValues {PlayerControllerIndex=0, bRequireCurrentFrameData=false}}) {module=CalcCameraFacing}
   &ScaleColor(Scale Alpha=2.5, ScaleA=true, ScaleRGB=true)
  ]
  rendererClasses [&NiagaraMeshRendererProperties]
  renderers [{rendererClass=&NiagaraMeshRendererProperties, rendererIndex=0}]
 }
 Orbit {
  emitterUpdateScript [
   &EmitterState(Life Cycle Mode=::NewEnumerator1, Loop Behavior=::NewEnumerator1, Loop Duration Mode=::NewEnumerator1, UseLoopDelay=false)
   &SpawnBurst_Instantaneous(Use Spawn Probability=false, UseLoopCountLimit=false)
  ]
  particleSpawnScript [
   &InitializeParticle(
    Color=<rgba 1 1 1 2>
    Lifetime=5
    Mesh Scale=&Multiply_VectorByFloat(Vector=<1 1 1>, Float=1.22)
    Mesh Scale Mode=::NewEnumerator4
    Position Offset=&AddVectorToPosition(Position=Linked() {linkedVariable {name=System.FXOffset, type=Vector3f}}, Vector=<0 0 0>)
    UsePositionOffset=true
    Write Material Random=true
   )
   &DynamicMaterialParameters(
    Index 0 Param 1=&Multiply_Float(A=1, B=$System.SizeAll)
    Module.Index0Float0Write=true
    Module.Index0Float1Write=true
    Module.Index0Float2Write=true
    Module.Index0Float3Write=true
   )
   &InitialMeshOrientation(Mesh Orientation Mode=::NewEnumerator1, Rotation=<0 0 0.5>, Use Rotation Vector=true) {module=InitialMeshOrientation003}
   &InitialMeshOrientation(Mesh Orientation Mode=::NewEnumerator7, Rotation=<0 0.5 0>, Use Rotation Vector=true) {module=InitialMeshOrientation004}
   &InitialMeshOrientation(Mesh Orientation Mode=::NewEnumerator7, Rotation=<0.04 0 0>, Use Rotation Vector=true) {module=InitialMeshOrientation005}
   &CameraOffset(Camera Offset Amount=5)
   &FN_Mesh_Attributes(
    Color=&MakeLinearColorFromVectorAndFloat(
     "Vector (RGB)"=&MakeVectorFromLinearColorRGB(Color=<rgba 0.0697 0.0045 0.2 0.9>)
     "Float (Alpha)"=&Multiply_Float(A=0.9, B=1.7)
    )
    Create Color Attribute=true
    Create Scale Attribute=true
    Module Output=::NewEnumerator1
    Scale=&Multiply_Vector(
     A=&Multiply_VectorByFloat(Vector=<1 1 1>, Float=1.15)
     B=&Multiply_VectorByFloat(Vector=&Multiply_VectorByFloat(Vector=<1 1 1>, Float=0.9), Float=$System.SizeAll)
    )
   )
  ]
  particleUpdateScript [
   &ParticleState(Kill Particles When Lifetime Has Elapsed=false)
   &ScaleMeshSize(Scale Factor=&Multiply_VectorByFloat(Vector=&Multiply_VectorByFloat(Vector=<1 1 1>, Float=0.9), Float=$System.SizeAll))
   &ScaleColor(Scale Alpha=1.5, Scale RGB=&Multiply_VectorByFloat(Vector=<1 1 1>, Float=1), ScaleA=true, ScaleRGB=true)
   &self:CalcCameraFacing(Camera Query=DataInterface() {propertyValues {PlayerControllerIndex=0, bRequireCurrentFrameData=false}}) {module=CalcCameraFacing}
  ]
  rendererClasses [&NiagaraMeshRendererProperties]
  renderers [
   {rendererClass=&NiagaraMeshRendererProperties, rendererIndex=0}
   {rendererClass=&NiagaraMeshRendererProperties, rendererIndex=1}
  ]
 }
 Stone001 {
  simTarget=GPUComputeSim
  emitterUpdateScript [
   &EmitterState(Life Cycle Mode=::NewEnumerator1, Loop Duration=1, UseLoopDelay=false)
   &SpawnRate(SpawnRate=9, Use Spawn Probability=false)
   &SpawnBurst_Instantaneous(Spawn Count=9, Use Spawn Probability=false, UseLoopCountLimit=false)
  ]
  particleSpawnScript [
   &InitializeParticle(
    Color=<rgba 0.7734 0.7656 1 1>
    Lifetime Max=5
    Lifetime Min=3
    Lifetime Mode=::NewEnumerator1
    Mesh Scale Mode=::NewEnumerator1
    Mesh Uniform Scale Max=1.5
    Mesh Uniform Scale Min=0.5
    Sprite Size Mode=::NewEnumerator3
    UsePositionOffset=false
    Write Material Random=true
   )
   &ShapeLocation(
    MaskSpawnGroup=false
    Offset=&AddVectorToPosition(Position=<0 0 0>, Vector=$System.FXOffset)
    Offset Mode=::NewEnumerator0
    Override Seed=false
    Rotation Mode=::NewEnumerator2
    Sphere Radius=16
    UseOwnerScale=true
   )
   &InitialMeshOrientation(Use Rotation Vector=false)
  ]
  particleUpdateScript [
   &SkeletalMeshLocation(
    Bone Sampling Mode=::NewEnumerator4
    MaskSpawnGroup=false
    Position Sampling=::NewEnumerator2
    Skeletal Mesh=DataInterface() {propertyValues {
      DefaultMesh=null
      ExcludeBoneName=root
      OptionalComponentClass=null
      PreviewMesh=/Game/Accessories/FORT_Backpacks/M_MED_Celestial_Backpack/M_MED_Celestial
      SkinningMode=SkinOnTheFly
      SoftSourceActor=null
      SourceMode=Default
      UvSetIndex=0
      WholeMeshLOD=-1
      bExcludeBone=false
      bReadDeformedGeometry=true
      bRequireCurrentFrameData=true
      ComponentTags []
      FilteredBones [M_MED_Celestial]
      FilteredSockets []
      MeshUserParameter {parameter {name=None, typeDefHandle {registeredTypeIndex=1}, varData []}}
      SamplingRegions []
     }}
   )
   &ParticleState()
   &ScaleMeshSize(Scale Factor=&Multiply_VectorByFloat(
     Vector=&RandomRangeVector(Minimum=<0.5 0.5 0.5>, Maximum=<1.2 1 1>)
     Float=&Multiply_Float(
      A=&Multiply_Float(
       A=0.12
       B=&FloatFromCurve(FloatCurve=DataInterface() {propertyValues {Curve {
           defaultValue=1
           keys [
            {interpMode=RCIM_Cubic, tangentMode=RCTM_User, time=0, value=0}
            {arriveTangent=1.4286, interpMode=RCIM_Cubic, leaveTangent=1.4286, time=0.3, value=1}
            {interpMode=RCIM_Cubic, tangentMode=RCTM_User, time=0.7, value=1}
            {tangentMode=RCTM_User, time=1, value=0}
           ]
          }}})
      )
      B=$System.InterpolateOverTime.InterpolateOverTime_Float.Moving Average
     )
    ))
   &VortexForce(
    UseOriginOffset=false
    Vortex Force Amount=300
    Vortex Origin:NiagaraPosition=&AddVectorToPosition(Position=$Engine.Emitter.SimulationPosition, Vector=$System.FXOffset)
   )
   &ScaleVelocity()
   /SetVariables(Transient.PhysicsForce:Vector3f=&Multiply_Vector(A=$Transient.PhysicsForce, B=<1 1 1>)) {isSetParameters=true}
   &PointAttractionForce(
    Attraction Strength=0.5
    Attractor Position:NiagaraPosition=&AddVectorToPosition(Position=$Engine.Emitter.SimulationPosition, Vector=$System.FXOffset)
    Kill Within Radius=false
    Module.Use Falloff=false
    UseAttractorOffset=false
   )
   &Drag(Drag=10, Rotational Drag=10, Use Linear Drag=true, Use Rotational Drag=true)
   &SolveForcesAndVelocity(Clamp Velocity=false, Limit Acceleration=false)
   &CameraOffset(Camera Offset Amount=5)
   &KillParticlesInVolume(Origin Offset:Vector3f=$System.FXOffset, Sphere Radius=12)
  ]
  rendererClasses [&NiagaraMeshRendererProperties]
  renderers [{rendererClass=&NiagaraMeshRendererProperties, rendererIndex=0}]
 }
}
summary {
 systemName=NS_Backpack_SkyRuled
 emitters [
  {
   emitterName=CenterRing
   rendererClasses [
    &NiagaraMeshRendererProperties
    &NiagaraSpriteRendererProperties
    &NiagaraLightRendererProperties
   ]
  }
  {emitterName=BackgorundRing, rendererClasses [&NiagaraMeshRendererProperties]}
  {emitterName=Stone001, simTarget=GPUComputeSim, rendererClasses [&NiagaraMeshRendererProperties]}
  {emitterName=Orbit, rendererClasses [&NiagaraMeshRendererProperties]}
  {emitterName=ColorRing, rendererClasses [&NiagaraMeshRendererProperties]}
 ]
 userVariables [
  User.bIsFrontEnd {type=NiagaraBool}
  User.bIsFrontEndPreview {type=NiagaraBool}
 ]
}
